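# Sound Cue for the recording beep sound effect used by Battle Pass Season 22 quests.
# A single Wave Player feeds the beep wave straight to the output at boosted volume, with no randomization or modulation.
# It borrows the Season 22 quest attenuation asset from the VO folder and routes through the non-critical in-game UI sound class.
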SoundCue /BattlePassS22/Audio/SFX/Quests_S22_RecordingBeep_Cue
root=WavePlayer_0
nodes [{id=WavePlayer_0, properties {soundWaveAssetPtr=&/Quests_S22_RecordingBeep}}]
properties {
 attenuationSettings=/BattlePassS22/Audio/VO/S22_Quest_Att
 maxDistance=2097152
 soundClassObject=&Fort_UI_InGame_NonCritical
 volumeMultiplier=1.5
}
